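; Swarm Queen boss pawn with throat and mouth weakpoints, reinforcement waves, crater bases and burrow relocation.
; Server-side perception picks random targets for the focus component, while interface events start the State Tree and drive repositioning and reinforcements.
; Reinforcements use Living World encounters chosen by playlist team size, tracked to gate new waves.
; Damage is capped per hit and multiplied on weakpoints; death dissolves materials, optionally despawns minions and credits attackers' quests.
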(event Custom|Play%20Hit%20Sound
  ; Play Hit Sounds
  (Networking|CanExecuteCosmeticEvents
    (:True
      (bind _damage_instigator (Variables|Default|GetDamage%20Instigator)
        (:then
          (if (Pawn|IsLocallyViewed (Pawn|GetControlledPawn _damage_instigator))
            (Audio|PlaySound2D (select (Variables|Default|GetIsWeakpointHit) (Variables|Default|GetWeakpoint%20Hit%20Sound) (Variables|Default|GetNormal%20Hit%20Sound)))))
        (:else)))))

(event Custom|OnDamaged_Event
  (params (DamagedActor :type Actor%20Object%20Reference) (Damage :type Float%20%28single-precision%29) (InstigatedBy :type Controller%20Object%20Reference) (DamageCauser :type Actor%20Object%20Reference) (HitLocation :type Vector) (FHitComponent :type Primitive%20Component%20Object%20Reference) (BoneName :type Name) (Momentum :type Vector))
  ; Add any pawn that damages the Queen to an array, that will be used to send a custom verb for quest completion
  (bind _output_get (Variables|Default|SetDamage%20Instigator InstigatedBy))
  (Utilities|Array|AddUnique (Variables|Default|GetAllDamagingPawns) _output_get)
  (CallFunction|PlayHitSound)
  (if (Variables|Default|GetIsWeakpointHit)
    (CallFunction|DoWeakpointHitFX)))

(event Custom|DoWeakpointHitFX
  (bind _skeletalmeshcomponentbudgeted (Variables|Default|GetSkeletalMeshComponentBudgeted))
  ; Darken weakpoint when it gets hit
  (Networking|CanExecuteCosmeticEvents
    (:True
      (Rendering|Material|SetCustomPrimitiveDataFloat (Variables|Default|GetSkeletalMeshComponentBudgeted) 4 (Utilities|Time|GetGameTimeinSeconds))
      ;[1 0.8441 0.1573 0.5] Weakpoint Hit FX
      (if (Variables|Default|GetIsThroatWeakpoint)
        (bind _returnvalue (Niagara|SpawnSystemAttached "/TidalCrane_Swarm_Boss/Assets/FX/NS_Weakspot_Impact.NS_Weakspot_Impact" _skeletalmeshcomponentbudgeted "neck_02" "0.000000,65.000000,0.000000"))
        (Transformation|SetWorldScale3D _returnvalue "1.000000,1.000000,1.000000")
        (else
          (bind _returnvalue_2 (Niagara|SpawnSystemAttached "/TidalCrane_Swarm_Boss/Assets/FX/NS_Weakspot_Impact.NS_Weakspot_Impact" _skeletalmeshcomponentbudgeted "fx_socket_mouth" "20.000000,0.000000,0.000000" "0.000000,-90.000000,0.000000"))
          (Transformation|SetWorldScale3D _returnvalue_2 "1.000000,1.000000,1.000000"))))))

(event Custom|SendQuestCustomVerb
  ; Add any pawn that damages the Queen to an array, that will be used to send a custom verb for quest completion
  (for _array_element (Variables|Default|GetAllDamagingPawns)
    (if (Utilities|IsValid _array_element :!class /Script/Engine.KismetSystemLibrary)
      (CustomVerb|BroadcastCustomVerbMessage (CustomVerb|MakeCustomVerbMessage _array_element "(TagName=\"CustomVerb.DamagedQueen\")")))))

(event EventBeginPlay
  (CallFunction|InitBoss)
  (CallFunction|DrawDebugBoss)
  (Networking|SwitchHasAuthority
    (:Authority
      (bind (_gamestate _playlist) (Class|FortAsyncActionCurrentPlaylistReady|PlaylistDataReadyAsync)
        (:OnReady
          (Variables|InternalData|SetCachedPlaylistTeamSize (Variables|/Script/FortniteGame.FortPlaylist|GetMaxTeamSize _playlist))
          (CallFunction|StartPerception)
          (Utilities|FlowControl|DelayUntilNextTick)
          (Abilities|TryActivateAbilitiesbyTag (Ability|GetAbilitySystemComponent self) (Variables|Default|GetTag_SpawnInAbility) false)
          (EventDispatchers|AssignOnDamaged (delegate Custom|OnDamaged_Event)))))
    (:Remote
      (AimAssist|RegisterAimAssistTarget 0 (Variables|Default|GetThroatWeakpoint) false)
      (Camera|PlayWorldCameraShake "/ProtoSwarm/Blueprints/EnvironmentalGameplay/B_CameraShake_Boss_Roar.B_CameraShake_Boss_Roar_C" (Transformation|GetActorLocation self) 1e+03 1e+04)
      (Utilities|FlowControl|Delay (ScalableFloat|GetValueatLevel (Variables|Default|GetVisibilityFallback)))
      (Rendering|SetActorHiddenInGame false)
      ; Store MIDs for death dissolve
      (Rendering|Material|SetCustomPrimitiveDataFloat (Variables|Default|GetSkeletalMeshComponentBudgeted) 2 0.0)
      (Utilities|Array|Clear (Variables|Default|GetMID_MaskedDeathMaterials))
      (for _array_element (Variables|Default|GetMaskedDeathMaterials)
        (bind _returnvalue (Rendering|Material|CreateDynamicMaterialInstance _array_element "None" "None" :!class /Script/Engine.KismetMaterialLibrary))
        (Utilities|Array|Add (Variables|Default|GetMID_MaskedDeathMaterials) _returnvalue)))))

(event Custom|MoveToNewLocation
  (params (NewLocation :type Vector))
  (Variables|InternalData|SetPostBurrowLocation NewLocation))

(event Custom|StartBrain
  ; Start Logic / State Tree / Perception - Triggered from the Intro Ability
  (AI|Logic|StartLogic (Variables|Default|GetStateTree)))

(event Custom|StartPerception
  (bind _fortnearbyactorsperception (Variables|Default|GetFortNearbyActorsPerception))
  ; Start Logic / State Tree / Perception - Triggered from the Intro Ability
  (Components|Activation|Activate _fortnearbyactorsperception false)
  (Query|BPRegisterListener _fortnearbyactorsperception (Utilities|Struct|MakeFortNearbyActorsPerceptionConfiguration 1.0 7 0 (ScalableFloat|GetValueatLevel (Variables|Default|GetPerceptionDistance)) "Friendly" false (Variables|Perception|GetPerceptionDetectGameplayTagQuery)) (EventDispatchers|CreateEvent :!event Custom|AIPerceptionUpdated)))

(event Custom|SpawnBase
  ; Spawn the Crater Actor
  (if (Variables|Default|GetSpawnBaseActor)
    (Game|SpawnActorfromClass (Variables|Default|GetBaseActorClass) (Math|Transform|MakeTransform (.location (Transformation|GetActorTransform)) (.rotation (Transformation|GetActorTransform)) (Math|Vector|MakeVector (ScalableFloat|GetValueatLevel (Variables|Default|GetBaseScaleSize)) (ScalableFloat|GetValueatLevel (Variables|Default|GetBaseScaleSize)) (ScalableFloat|GetValueatLevel (Variables|Default|GetBaseScaleSize)))) "AlwaysSpawn" "MultiplyWithRoot" self 0 0)))

(event Custom|SetupBase
  (bind _returnvalue (Transformation|GetActorLocation))
  (bind _self self)
  ; Spawn Actors Required for the Base/Crater at the feet
  (Networking|SwitchHasAuthority
    (:Authority
      (CallFunction|SpawnBase)
      (if (Variables|Default|GetSpawnBaseActor)
        (bind _returnvalue_2 (Game|SpawnActorfromClass "/TidalCrane_Swarm_Boss/Blueprints/BP_TidalCrane_Boss_RepositionLocations.BP_TidalCrane_Boss_RepositionLocations_C" (Math|Transform|MakeTransform _returnvalue) "AlwaysSpawn" "MultiplyWithRoot" _self false 0)) ; Spawn Relocation Point With Visual
        (Utilities|Array|Add (Variables|InternalData|GetRepositionActors) _returnvalue_2)
        (Utilities|Map|Add (Variables|InternalData|GetRepositionActorsAndBases) _returnvalue_2 "true")
        (else
          (bind _returnvalue_3 (Game|SpawnActorfromClass "/TidalCrane_Swarm_Boss/Blueprints/BP_TidalCrane_Boss_RepositionLocations_NoVisuals.BP_TidalCrane_Boss_RepositionLocations_NoVisuals_C" (Math|Transform|MakeTransform _returnvalue) "AlwaysSpawn" "MultiplyWithRoot" _self false 0)) ; Spawn Relocatoin Point Without Visual
          (Utilities|Array|Add (Variables|InternalData|GetRepositionActors) _returnvalue_3)
          (Utilities|Map|Add (Variables|InternalData|GetRepositionActorsAndBases) _returnvalue_3 "true"))))))

(event Custom|SpawnDustEffects
  ; Spawn Dust Effects
  (Niagara|SpawnSystematLocation "/ProtoSwarm/Assets/FX/NS_CentralPicnic_Milkshake_Burst.NS_CentralPicnic_Milkshake_Burst" (Transformation|GetActorLocation self)))

(event Custom|AIPerceptionUpdated
  (params (AllFoundActors :type Array%20of%20Actor%20Object%20References) (EnteringActors :type Array%20of%20Actor%20Object%20References) (ExitingActors :type Array%20of%20Actor%20Object%20References))
  ; Track Nearby Pawns / Get Target
  (Variables|Default|SetAllNearbyPawns AllFoundActors)
  (Utilities|FlowControl|Sequence
    (:then_0
      (Utilities|IsValid (Variables|InternalData|GetCurrentTarget)
        (:Is%20Not%20Valid
          (Default|SetCurrentTarget (Utilities|Array|RandomArrayItem AllFoundActors))
          (Variables|InternalData|SetCanTriggerReinforcements true))))
    (:then_1
      (bind _array_element (Utilities|Array|ForEachLoopwithBreak ExitingActors)
        (:LoopBody
          (if (== _array_element (Variables|InternalData|GetCurrentTarget))
            (if (>= (Utilities|Array|Length AllFoundActors) 1)
              (Default|SetCurrentTarget (Utilities|Array|RandomArrayItem AllFoundActors))
              (Variables|InternalData|SetCanTriggerReinforcements true))
            (Default|SetCurrentTarget 0) ; No Nearby actors
            (Variables|InternalData|SetCanTriggerReinforcements false)))
        (:Completed
          _))
      (Variables|InternalData|SetPotentialTargets AllFoundActors))))

(event Custom|FindAllNearbyRelocationActors
  ; Find Actors where we want the Queen to reposition at
  (Networking|SwitchHasAuthority
    (:Authority
      (bind _outactors (Collision|SphereOverlapActors (Transformation|GetActorLocation) (ScalableFloat|GetValueatLevel (Variables|Default|GetRelocationActorDetectionRadius)) (Utilities|Array|MakeArray "ObjectTypeQuery1" "ObjectTypeQuery2") "/TidalCrane_Swarm_Boss/Blueprints/BP_TidalCrane_Boss_RepositionLocations.BP_TidalCrane_Boss_RepositionLocations_C"))
      (bind _output_get (Variables|InternalData|SetRepositionActors _outactors))
      (for _array_element _output_get
        (Utilities|Map|Add (Variables|InternalData|GetRepositionActorsAndBases) _array_element)))))

(event Custom|HandleSpawnBaseActor
  (bind _repositionactorsandbases (Variables|InternalData|GetRepositionActorsAndBases))
  ; Check if we should be spawning a new base or not at the new relocation location.
  (if (Variables|Default|GetSpawnBaseActor)
    (Networking|SwitchHasAuthority
      (:Authority
        (for _array_element (Variables|InternalData|GetRepositionActors)
          (if (Utilities|Operators|Equal(==) (Transformation|GetActorLocation _array_element) (Transformation|GetActorLocation) 2e+02)
            (if (.ReturnValue (Utilities|Map|Find _repositionactorsandbases _array_element))
              (if (not (Utilities|Map|Find _repositionactorsandbases _array_element))
                (Utilities|Map|Add _repositionactorsandbases _array_element "true")
                (CallFunction|SpawnBase)))))))))

(event EventOnDeathPlayEffects
  (params (Damage :type Float%20%28single-precision%29) (DamageTags :type Gameplay%20Tag%20Container%20Structure%20%28by%20ref%29) (Momentum :type Vector :default "0, 0, 0") (HitInfo :type Hit%20Result%20Structure%20%28by%20ref%29) (InstigatedBy :type Fort%20Pawn%20Object%20Reference) (DamageCauser :type Actor%20Object%20Reference) (EffectContext :type Gameplay%20Effect%20Context%20Handle%20Structure))
  ; On Death
  (Networking|CanExecuteCosmeticEvents
    (:True
      (Animation|Montage|PlayMontage (Variables|Default|GetSkeletalMeshComponentBudgeted) (Variables|Default|GetEliminationMontage))
      (GameplayCue|ExecuteGameplayCue%28Burst%29Local self (Variables|Default|GetGC_BossDeath) (Ability|GameplayCue|MakeGameplayCueParameters :NormalizedMagnitude 0.0 :RawMagnitude 0.0 :OriginalTag (Variables|Default|GetGC_BossDeath) :Location "0,0,0" :Normal "0,0,0" :GameplayEffectLevel 1 :AbilityLevel 1 :bReplicateLocationWhenUsingMinimalRepProxy false) false)
      (CallFunction|DeathDissolve)
      (Components|Activation|Deactivate (Variables|Default|GetNS_Queen_Drool)))))

(event Custom|TriggerReinforcements
  (params (SmallReinforcements :type Boolean))
  (bind _encounterref (Variables|InternalData|GetEncounterRef))
  ; Reinforcements Tracker
  (if SmallReinforcements
    (CallFunction|TriggerReinforcementSpit)
    (Utilities|FlowControl|Delay 0.2)
    (bind _asfort_athena_living_world_static_point_provider (Utilities|Casting|CastToFortAthenaLivingWorldStaticPointProvider (Variables|/Script/Engine.ChildActorComponent|GetChildActor (Variables|Default|GetLW_StaticPoint))))
    (Class|FortAthenaLivingWorldPointProviderInterface|EnablePointProvider_DEPRECATED _asfort_athena_living_world_static_point_provider)
    (Utilities|FlowControl|DelayUntilNextTick)
    (bind (_isvalid _returnvalue) (Default|StartActivityEncounterStaticPointProvider (Game|GetGameState) self (CallFunction|GetEncounterType) (Utilities|Array|MakeArray _asfort_athena_living_world_static_point_provider)))
    (Variables|InternalData|SetEncounterRef _returnvalue)
    (else
      (bind _asfort_athena_living_world_volume (Utilities|Casting|CastToFortAthenaLivingWorldVolume (Variables|/Script/Engine.ChildActorComponent|GetChildActor (Variables|Default|GetLivingWorldVolume_ChildActor))))
      (Class|FortAthenaLivingWorldPointProviderInterface|EnablePointProvider_DEPRECATED _asfort_athena_living_world_volume)
      (Utilities|FlowControl|Delay 0.5)
      (Class|FortAthenaLivingWorldManager|RequestEventGeneration (Class|FortAthenaLivingWorldManager|GetLivingWorldManager))
      (bind (_isvalid_2 _returnvalue_2) (Default|StartActivityEncounter (Game|GetGameState) _asfort_athena_living_world_volume self (CallFunction|GetEncounterType)))
      (Variables|InternalData|SetEncounterRef _returnvalue_2)))
  (EventDispatchers|BindEventtoOnLastActorDied _encounterref (EventDispatchers|CreateEvent :!event Custom|OnEncounterLastActorDied))
  (EventDispatchers|BindEventtoOnActorSpawned _encounterref (EventDispatchers|CreateEvent :!event Custom|EncounterActorSpawned))
  (Variables|InternalData|SetCanTriggerReinforcements false))

(event Custom|OnEncounterLastActorDied
  (params (EncounterInstance :type Fort%20Athena%20Living%20World%20Encounter%20Instance%20Object%20Reference))
  ; Reinforcements Tracker
  (Variables|InternalData|SetCurrentReinforcementsCount 0)
  (Utilities|FlowControl|Delay (Math|Random|RandomFloatinRange (ScalableFloat|GetValueatLevel (Variables|Default|GetReinforcements_MinDelay)) (ScalableFloat|GetValueatLevel (Variables|Default|GetReinforcements_MaxDelay))))
  (Variables|InternalData|SetCanTriggerReinforcements true)
  (bind _asfort_athena_living_world_volume (Utilities|Casting|CastToFortAthenaLivingWorldVolume (Variables|/Script/Engine.ChildActorComponent|GetChildActor (Variables|Default|GetLivingWorldVolume_ChildActor))))
  (Class|FortAthenaLivingWorldVolume|DisablePointProvider _asfort_athena_living_world_volume))

(event Custom|DeathDissolve
  (bind _skeletalmeshcomponentbudgeted (Variables|Default|GetSkeletalMeshComponentBudgeted))
  ;[0.1315 0.03596 0.15 0.5] DeathDissolve: Execute the death dissolve material update.
  (Networking|CanExecuteCosmeticEvents
    (:True
      (for (_array_element _array_index) (Variables|Default|GetMID_MaskedDeathMaterials)
        (Rendering|Material|SetMaterial _skeletalmeshcomponentbudgeted _array_index _array_element) ; Swap to masked material on death
      )
      (Rendering|Material|SetCustomPrimitiveDataVector2 _skeletalmeshcomponentbudgeted 2 (Math|Vector2D|MakeVector2D 1.0 (Utilities|Time|GetGameTimeinSeconds))))))

(event Custom|TriggerReinforcementSpit
  ; Reinforcements Tracker
  (Networking|CanExecuteCosmeticEvents
    (:True
      (GameplayCue|ExecuteGameplayCue%28Burst%29Local self (Variables|Default|GetGC_ReinforcementSpit) (Ability|GameplayCue|MakeGameplayCueParameters :NormalizedMagnitude 0.0 :RawMagnitude 0.0 :OriginalTag (Variables|Default|GetGC_ReinforcementSpit) :Location (Transformation|GetActorLocation (Variables|/Script/Engine.ChildActorComponent|GetChildActor (Variables|Default|GetLW_StaticPoint))) :Normal "0,0,0" :GameplayEffectLevel 1 :AbilityLevel 1 :bReplicateLocationWhenUsingMinimalRepProxy false)))))

(event Custom|EncounterActorSpawned
  (params (EncounterInstance :type Fort%20Athena%20Living%20World%20Encounter%20Instance%20Object%20Reference) (SpawnedActor :type Actor%20Object%20Reference) (EventData :type Fort%20Athena%20Living%20World%20Event%20Data%20Object%20Reference))
  ; Reinforcements Tracker
  (Math|Integer|IncrementInt (Variables|InternalData|GetCurrentReinforcementsCount)
    (:%20%20
      (bind _asfort_aipawn (Utilities|Casting|CastToFortAIPawn SpawnedActor))
      (Class|BPFLAICommonFunctions|SetGoalActor _asfort_aipawn (Variables|InternalData|GetCurrentTarget)))))

(event EventOnDeathServer
  (params (Damage :type Float%20%28single-precision%29) (DamageTags :type Gameplay%20Tag%20Container%20Structure%20%28by%20ref%29) (Momentum :type Vector :default "0, 0, 0") (HitInfo :type Hit%20Result%20Structure%20%28by%20ref%29) (InstigatedBy :type Controller%20Object%20Reference) (DamageCauser :type Actor%20Object%20Reference) (EffectContext :type Gameplay%20Effect%20Context%20Handle%20Structure))
  ; On Death Server
  (CallFunction|AllowLookatTarget false)
  (CallFunction|SendQuestCustomVerb)
  (CallFunction|ForceDespawnSpawnedSwarm))

(event Custom|ForceDespawnSpawnedSwarm
  ; Force Despawn Minions on Queen Death
  (if (ScalableFloat|GetValueasBool (Variables|Default|GetShouldDespawnMinionsOnDeath))
    (for _array_element (Variables|/Script/LagerRuntime.FortAthenaLivingWorldEncounterInstance|GetSpawnedActors (Variables|InternalData|GetEncounterRef))
      (Ability|SendGameplayEventtoActor _array_element "(TagName=\"Event.TidalCrane.Swarm.ForceDespawn\")"))))

(event Custom|ForceLookAtLocationOnSpawn
  ; Control Pause/Unpause Focus
  (Transformation|SetActorRotation self (Math|Rotator|MakeRotator (.roll (Transformation|GetActorRotation self)) (.pitch (Transformation|GetActorRotation self)) (.yaw (Math|Rotator|FindLookatRotation (Transformation|GetActorLocation self) (Transformation|GetActorLocation (Variables|InternalData|GetCurrentTarget))))) false))

(event Custom|AllowLookAtTarget
  (params (Allow :type Boolean))
  (bind _fortactorfocus (Variables|Default|GetFortActorFocus))
  ; Control Pause/Unpause Focus
  (if Allow
    (if (Health|HasHealthLeft self)
      (Class|FortActorFocusComponent|UnpauseFocus _fortactorfocus))
    (else
      (Class|FortActorFocusComponent|PauseFocus _fortactorfocus))))

(event Custom|TidalCraneQueen_StartTakeOver
  ;[0.6562 0.5853 0.04086 0.5] Cheat : AITakeOver-ish
  (Development|PrintString "Disabled AI TAke Over to save some memory" true true "(R=0.000000,G=0.660000,B=1.000000,A=1.000000)" 1e+01)
  (Fortnite|FortShippingLog self "Disabled AI TAke Over to save some memory" true))

(event Default|EventSetAllowLookAtTarget
  (params (AllowLookAt :type Boolean))
  ;[0 0.15 0.1452 0.5] Interface Sets
  (CallFunction|AllowLookatTarget AllowLookAt))

(event Default|EventHandleSetupBase
  ;[0 0.15 0.1452 0.5] Interface Sets
  (CallFunction|SetupBase))

(event EventHandleNewTargetSelection
  ;[0 0.15 0.1452 0.5] Interface Sets
  (CallFunction|SelectNewTarget))

(event Default|EventHandleForceLookAtLocationOnSpawn
  ;[0 0.15 0.1452 0.5] Interface Sets
  (CallFunction|ForceLookatLocationonSpawn))

(event EventSetProjectileVolleySpawnLocations
  (params (ProjectileVolleySpawnLocations :type Array%20of%20Vectors))
  ;[0 0.15 0.1452 0.5] Interface Sets
  (Variables|InternalData|SetProjectileVolleyLocations ProjectileVolleySpawnLocations))

(event Default|EventHandleStartBrain
  ;[0 0.15 0.1452 0.5] Interface Sets
  (CallFunction|StartBrain))

(event Default|EventHandleSpawnCraterBaseActor
  ;[0 0.15 0.1452 0.5] Interface Sets
  (CallFunction|HandleSpawnBaseActor))

(event Default|EventHandleTriggerReinforcements
  (params (SmallReinforcement :type Boolean))
  ;[0 0.15 0.1452 0.5] Interface Sets
  (CallFunction|TriggerReinforcements SmallReinforcement))

(event Default|EventHandleMoveToNewLocation
  (params (NewLocation :type Vector :default "0, 0, 0"))
  ;[0 0.15 0.1452 0.5] Interface Sets
  (CallFunction|MovetoNewLocation NewLocation))

(event EventSetCurrentTargetActor
  (params (NewTarget :type Actor%20Object%20Reference))
  ;[0 0.15 0.1452 0.5] Interface Sets
  (Default|SetCurrentTarget NewTarget))

(event EventEndPlay
  (params (EndPlayReason :type EEndPlayReason%20Enum :default "Destroyed"))
  (Networking|SwitchHasAuthority
    (:Remote
      (AimAssist|UnregisterAimAssistTarget 0))))

(fn BlueprintModifyIncomingDamage
  (params (Damage :type Float%20%28single-precision%29) (DamageTags :type Gameplay%20Tag%20Container%20Structure%20%28by%20ref%29) (EffectContext :type Gameplay%20Effect%20Context%20Handle%20Structure%20%28by%20ref%29) (EventInstigator :type Controller%20Object%20Reference) (DamageCauser :type Actor%20Object%20Reference))
  (bind _returnvalue (Ability|EffectContext|GetHitResult EffectContext))
  (bind (_bblockinghit _binitialoverlap _time _distance _location _impactpoint _normal _impactnormal _physmat _hitactor _hitcomponent) (Collision|BreakHitResult _returnvalue))
  (bind _mouthweakpoint (Variables|Default|GetMouthWeakpoint))
  (bind _returnvalue_2 (== _hitcomponent _mouthweakpoint))
  (bind _throatweakpoint (Variables|Default|GetThroatWeakpoint))
  (bind _returnvalue_3 (== _hitcomponent _throatweakpoint))
  (bind _returnvalue_4 (or _returnvalue_3 _returnvalue_2))
  (LocalVariables|SetCanIgnoreMaxDmgCap (GameplayTags|HasTag DamageTags "(TagName=\"TidalCrane.AI.Boss.Dmg.IgnoreMaxCap\")"))
  (Ability|EffectContext|SetCritical (Variables|GetEffectContext) _returnvalue_4)
  (bind _output_get (Variables|Default|SetIsWeakpointHit _returnvalue_4))
  (if _output_get
    (Variables|Default|SetIsThroatWeakpoint _returnvalue_3)
    (else
      _))
  (return (select _returnvalue_4 (* (Math|Float|Min(Float) (Variables|GetDamage) (select (LocalVariables|GetCanIgnoreMaxDmgCap) (Variables|GetDamage) (ScalableFloat|GetValueatLevel (Variables|Default|GetMaxDamageByHit)))) (ScalableFloat|GetValueatLevel (Variables|Default|GetWeakPointDamageMultiplier))) (Math|Float|Min(Float) (Variables|GetDamage) (select (LocalVariables|GetCanIgnoreMaxDmgCap) (Variables|GetDamage) (ScalableFloat|GetValueatLevel (Variables|Default|GetMaxDamageByHit)))))))

(fn DrawDebugBoss
  (bind _returnvalue (Transformation|GetActorLocation))
  (bind _returnvalue_2 (+ _returnvalue "0.000000,0.000000,1600.000000"))
  (if (Variables|Debug|GetDrawDebug)
    (Utilities|FlowControl|Sequence
      (:then_0
        ; Draw current orientation
        (Rendering|Debug|DrawDebugArrow _returnvalue_2 (+ _returnvalue_2 (* (Transformation|GetActorForwardVector) 8e+02)) 3e+02 (Math|Color|SelectColor "(R=1.000000,G=0.000000,B=0.020022,A=1.000000)" "(R=0.000000,G=0.000000,B=1.000000,A=1.000000)" (Networking|HasAuthority)) 0.0 15.0)
        (if (Class|FortActorFocusComponent|IsPaused (Variables|Default|GetFortActorFocus))
          (Rendering|Debug|DrawDebugString (+ _returnvalue_2 "0.000000,0.000000,200.000000") "FOCUS PAUSED" 0 "(R=1.000000,G=0.016621,B=0.000000,A=1.000000)")))
      (:then_1
        ; Draw current target
        (Networking|SwitchHasAuthority
          (:Authority
            (bind _currenttarget (Variables|InternalData|GetCurrentTarget)
              (:then
                (Rendering|Debug|DrawDebugLine (Transformation|GetActorLocation) (Transformation|GetActorLocation _currenttarget) "(R=1.000000,G=1.000000,B=1.000000,A=1.000000)" 0.0 3.0))
              (:else)))))
      (:then_2
        (Utilities|Time|SetTimerforNextTickbyEvent (EventDispatchers|CreateEvent :!event DrawDebugBoss))))
    (else
      (return))))

(fn GetAllNearbyPawns
  (return (Variables|Default|GetAllNearbyPawns)))

(fn GetAnimData
  (bind _currenttarget (Variables|InternalData|GetCurrentTarget))
  (bind _returnvalue_location (Math|Transform|BreakTransform (Transformation|GetActorTransform _currenttarget)))
  (Utilities|IsValid _currenttarget
    (:Is%20Valid
      (return (Transformation|GetActorTransform) _returnvalue_location))))

(fn GetCanTriggerReinforcements
  (return (Variables|InternalData|GetCanTriggerReinforcements)))

(fn GetCurrentReinforcementsCount
  (return (Variables|InternalData|GetCurrentReinforcementsCount)))

(fn GetCurrentTarget
  (return (Variables|InternalData|GetCurrentTarget)))

(fn GetCurrentTargetActor
  (bind _currenttargetactor (CallFunction|GetCurrentTarget))
  (return _currenttargetactor))

(fn GetEncounterType
  (return (select (Variables|Cheat|GetAITakeOverEnabled) (Variables|Default|Encounter|GetEncounter%20Type%20AITakeOver) (Utilities|Map|Find (Variables|Default|Encounter|GetReinforcementEncounters) (Variables|InternalData|GetCachedPlaylistTeamSize)))))

(fn GetNumberOfPotentialTargets
  (bind _potentialtargetscount (CallFunction|GetNumberPotentialTargets))
  (return _potentialtargetscount))

(fn GetNumberPotentialTargets
  (return (Utilities|Array|Length (Variables|InternalData|GetPotentialTargets))))

(fn GetPotentialTargets
  (return (Variables|InternalData|GetPotentialTargets)))

(fn GetProjectileVolleySpawnLocations
  (return (Variables|InternalData|GetProjectileVolleyLocations)))

(fn GetRepositionActors
  (return (Variables|InternalData|GetRepositionActors)))

(fn GetStateTree
  (return (Variables|Default|GetStateTree)))

(fn GetTargetOwner
  (return 0))

(fn GetWeakspotActors
  (return 0))

(fn GetWeakspotComponents
  (return (Utilities|Array|MakeArray (Variables|Default|GetMouthWeakpoint))))

(fn InitBoss
  (bind _astidal_crane_swarm_boss_anim_bp (Utilities|Casting|CastToTidalCrane_SwarmBoss_AnimBP (Components|SkeletalMesh|GetAnimInstance (Variables|Default|GetSkeletalMeshComponentBudgeted)))
    (:then
      (Variables|InternalData|SetBossAnimBP _astidal_crane_swarm_boss_anim_bp))
    (:CastFailed
      _))
  (CallFunction|FindAllNearbyRelocationActors))

(fn IsValidAutoFireTarget
  (params (Attacker :type Actor%20Object%20Reference))
  (return true))

(fn OnRep_CurrentTarget)

(fn OnRep_PostBurrowLocation
  (Transformation|SetActorLocation self (Variables|InternalData|GetPostBurrowLocation) false true))

(fn SelectNewTarget
  (for _array_element (Variables|InternalData|GetPotentialTargets)
    (if (!= _array_element (Variables|InternalData|GetCurrentTarget))
      (Utilities|Array|Add (LocalVariables|GetTempTargetsWithoutCurrentTarget) _array_element)))
  (bind _output_get (LocalVariables|SetRandomTarget (Utilities|Array|RandomArrayItem (LocalVariables|GetTempTargetsWithoutCurrentTarget))))
  (Default|SetCurrentTarget _output_get)
  (return (Utilities|IsValid _output_get :!class /Script/Engine.KismetSystemLibrary)))

(fn SetCurrentTarget
  (params (NewTarget :type Actor%20Object%20Reference))
  (Networking|SwitchHasAuthority
    (:Authority
      (bind _output_get (Variables|InternalData|SetCurrentTarget NewTarget))
      (Class|FortActorFocusComponent|SetActorFocus (Variables|Default|GetFortActorFocus) _output_get))))

(fn ShouldAlliesAutoFireOnDeployedTarget
  (return false))

(fn ShouldAutoFireOnSelfDeployedTarget
  (return false))

(fn ConstructionScript
  (Transformation|AttachComponentToComponent (Variables|Default|GetThroatWeakpoint) (Variables|Default|GetSkeletalMeshComponentBudgeted) "neck_02" "KeepRelative" "KeepRelative" "KeepRelative" false)
  (Transformation|AttachComponentToComponent (Variables|Default|GetMouthWeakpoint) (Variables|Default|GetSkeletalMeshComponentBudgeted) "head" "KeepRelative" "KeepRelative" "KeepRelative" false)
  (Transformation|AttachComponentToComponent (Variables|Default|GetBP_TidalCrane_Boss_BodyAimAssist) (Variables|Default|GetSkeletalMeshComponentBudgeted) "spine_01" "KeepRelative" "KeepRelative" "KeepWorld" false))
