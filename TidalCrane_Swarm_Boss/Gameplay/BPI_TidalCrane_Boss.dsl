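; Blueprint interface for the TidalCrane swarm boss, declaring the calls through which other code queries and drives the boss without implementing them.
; Getters expose the StateTree, current and potential targets, nearby pawns, reposition actors, volley spawn locations and reinforcement state.
; Handlers cover base setup, crater base spawning, brain start, spawn look-at, target selection, movement and small-flagged reinforcement triggers.
; Setters update the current target, look-at permission and projectile volley spawn locations.
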(fn GetAllNearbyPawns
  (return 0))

(fn GetCanTriggerReinforcements
  (return false))

(fn GetCurrentReinforcementsCount
  (return 0))

(fn GetCurrentTargetActor
  (return 0))

(fn GetNumberOfPotentialTargets
  (return 0))

(fn GetPotentialTargets
  (return 0))

(fn GetProjectileVolleySpawnLocations
  (return 0))

(fn GetRepositionActors
  (return 0))

(fn GetStateTree
  (return 0))

(fn HandleForceLookAtLocationOnSpawn)

(fn HandleMoveToNewLocation
  (params (NewLocation :type Vector)))

(fn HandleNewTargetSelection)

(fn HandleSetupBase)

(fn HandleSpawnCraterBaseActor)

(fn HandleStartBrain)

(fn HandleTriggerReinforcements
  (params (SmallReinforcement :type Boolean)))

(fn SetAllowLookAtTarget
  (params (AllowLookAt :type Boolean)))

(fn SetCurrentTargetActor
  (params (NewTarget :type Actor%20Object%20Reference)))

(fn SetProjectileVolleySpawnLocations
  (params (ProjectileVolleySpawnLocations :type Array%20of%20Vectors)))
